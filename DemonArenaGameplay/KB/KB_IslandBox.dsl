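; KB_IslandBox is a Blueprint that tags itself with gameplay tags when play starts.
; On BeginPlay it calls AddLooseGameplayTags on itself, passing the GameplayTags variable.
; The construction script is empty, and the file shows no other logic.
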(event EventBeginPlay
  (Ability|GameplayEffect|AddLooseGameplayTags self (Variables|Default|GetGameplayTags)))

(fn ConstructionScript)
