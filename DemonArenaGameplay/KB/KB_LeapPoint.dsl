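; KB_LeapPoint is a leap point marker actor for the demon arena gameplay.
; Its construction script traces 200 units straight down for objects and, on a hit, moves the actor to the impact point plus a hotfixable scalable-float height offset.
; On BeginPlay it adds its configured gameplay tags to itself as loose tags, presumably so other systems can identify it.
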(event EventBeginPlay
  (Ability|GameplayEffect|AddLooseGameplayTags self (Variables|Default|GetGameplayTags)))

(fn ConstructionScript
  (bind (_outhit _returnvalue) (Collision|LineTraceForObjects (Transformation|GetActorLocation) (- (Transformation|GetActorLocation) "0.000000,0.000000,200.000000") (Utilities|Array|MakeArray "ObjectTypeQuery1") true))
  (if _returnvalue
    (Transformation|SetActorLocation (Utilities|Operators|Add (.ImpactPoint (Collision|BreakHitResult _outhit)) (Math|Vector|MakeVector 0.0 0.0 (ScalableFloat|GetValueatLevel (Variables|Hotfix|GetLeapPointTraceHeight_Hf)))))))
